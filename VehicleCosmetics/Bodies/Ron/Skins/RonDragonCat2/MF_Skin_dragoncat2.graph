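# Material function for the DragonCat2 skin of the Ron vehicle body, adding animated, audio-reactive neon waves over the base color map.
# MF_AudioBeatSync supplies low, high and combined beat signals alongside texture masks and noise-based dot and gradient patterns.
# Two sine-timed, noise-warped UV panners with a zigzag offset form wave bands and edge lines.
# Layered lerps color these pink, blue, yellow and white into base color and emissive, with constant metallic and roughness.
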MaterialFunction /VehicleCosmetics/Bodies/Ron/Skins/RonDragonCat2/MF_Skin_dragoncat2
graph {nodes [
  n1 = TextureSampleParameter2D() {parameterName=BaseColorMap, texture=&/T_rondragoncat2_BC}
  n2 = Constant3Vector() {constant=<rgba 0.8589 0.0625 1 1>}
  n3 = TextureSampleParameter2D() {parameterName=RGBMask, texture=/VehicleCosmetics/Bodies/Ron/Skins/RonDragonCat1/T_RonDragoncat1_RGB}
  n4 = Constant() {r=40}
  n5 = Constant() {r=0.5}
  n6 = Constant() {r=2}
  n7 = NamedRerouteUsage()
  n8 = Multiply(
   A=@n3.A
   B=Multiply(
    A=MaterialFunctionCall!(Texture Coordinates, Tiling X=@n4, Tiling Y=@n4, Offset=@n5, Angle, Sharpen=@n6, PanSpeedX, PanSpeedY=@n5, RotateSpeed) {materialFunction=/VehicleCosmetics/Materials/MaterialFunctions/MF_Noisemaker_Dots}.Dots
    B=ConstantBiasScale(None=@n7)
   )
  )
  n9 = MaterialFunctionCall!(Texture Coordinates, Tiling=Constant() {r=16}, PanSpeed=Constant() {r=0.3}, Angle, RotateSpeed, Sharpen, Offset) {materialFunction=/VehicleCosmetics/Materials/MaterialFunctions/MF_Noisemaker_Gradient}
  n10 = Saturate(None=Multiply(A=ConstantBiasScale(None=@n9.Gradient2) {bias=-0.5, scale=0.75}, B=@n7))
  n11 = Multiply(A=Saturate(None=ConstantBiasScale(None=@n3.A) {bias=-0.1, scale=2}), B=ConstantBiasScale(None=Add(A=@n8, B=@n10)) {bias=-0.2, scale=6})
  n12 = Saturate(None=Subtract(A=Saturate(None=@n11), B=NamedRerouteUsage()))
  n13 = Constant3Vector() {constant=<rgba 0.0018 0.1652 1 1>}
  n14 = ComponentMask(None=TextureCoordinate()) {g=true}
  n15 = AppendVector(A=Constant(), B=Add(A=NamedRerouteUsage(), B=Multiply(A=@n14) {constB=2.1}))
  n16 = Multiply(A=Multiply(A=NamedRerouteUsage()) {constB=5}, B=ScalarParameter() {defaultValue=0.5, parameterName=BeatWavesStrength}) {constB=5}
  n17 = Constant() {r=10}
  n18 = Constant() {r=0.5}
  n19 = ConstantBiasScale(None=MaterialFunctionCall!(Texture Coordinates, Tiling X=@n17, Tiling Y=@n18, PanSpeedX=@n18, PanSpeedY=@n18, Chaos Speed, Angle, Rotation Speed, Sharpen=@n17) {materialFunction=/VehicleCosmetics/Materials/MaterialFunctions/MF_Noisemaker_AnisotropicSmooth}.Aniso2) {bias=0.01, scale=0.05}
  n20 = Time() {bOverride_Period=true, period=6000}
  n21 = Panner(Coordinate=Add(A=@n15, B=Multiply(A=@n16, B=@n19) {constB=5}), Time=Sine(None=@n20) {period=4}) {speedY=0.056}
  n22 = ComponentMask(None=TextureCoordinate() {coordinateIndex=1, uTiling=2, vTiling=0.4}) {r=true}
  n23 = OneMinus(None=@n22)
  n24 = Floor(None=@n22)
  n25 = ComponentMask(None=Panner(Coordinate=LinearInterpolate(A=@n22, B=@n23, Alpha=@n24), Time=@n20) {speedX=0.5, speedY=0.2}) {r=true}
  n26 = Subtract(A=@n25, B=Floor(None=@n25))
  n27 = ConstantBiasScale(None=Abs(None=ConstantBiasScale(None=@n26) {bias=-0.5, scale=2})) {bias=-0.5, scale=3}
  n28 = ConstantBiasScale(None=Multiply(A=Abs(None=@n27), B=@n27)) {bias=-0.5, scale=0.03}
  n29 = Add(A=Add(A=ComponentMask(None=@n21) {g=true}, B=@n28)) {constB=-0.2}
  n30 = Subtract(A=Ceil(None=@n29), B=@n29)
  n31 = ConstantBiasScale(None=MaterialFunctionCall(Tiling X=@n17, Tiling Y=Constant() {r=0.2}, PanSpeedX=Constant() {r=0.1}, PanSpeedY=@n18, Sharpen=@n17) {materialFunction=&MF_Noisemaker_AnisotropicSmooth}.Aniso2) {bias=0.01, scale=0.1}
  n32 = Panner(Coordinate=Add(A=@n15, B=Multiply(A=@n16, B=@n31) {constB=5}), Time=Sine(None=Add(A=@n20) {constB=15}) {period=5}) {speedY=0.07}
  n33 = Add(A=Add(A=ComponentMask(None=@n32) {g=true}, B=@n28)) {constB=0.05}
  n34 = Subtract(A=Ceil(None=@n33), B=@n33)
  n35 = Max(A=@n30, B=@n34)
  n36 = Multiply(A=ComponentMask(None=Panner(Coordinate=@n35, Time=@n20) {speedY=-0.6}) {g=true}) {constB=2}
  n37 = ConstantBiasScale(None=Subtract(A=@n36, B=Floor(None=@n36))) {bias=-0.5, scale=2}
  n38 = Abs(None=@n37)
  n39 = LinearInterpolate(B=ConstantBiasScale(None=Power(Base=@n38) {constExponent=0.2}) {bias=-0.45, scale=2}, Alpha=NamedRerouteUsage())
  n40 = LinearInterpolate(A=LinearInterpolate(A=@n2, B=@n13, Alpha=Saturate(None=@n39) {desc=out}) {desc=PinkBlueWaves}, Alpha=@n12) {constB=0, desc=PinkBlueWaves}
  n41 = Saturate(None=@n35) {desc=out}
  n42 = NamedRerouteUsage()
  n43 = LinearInterpolate(A=@n41, Alpha=@n42) {constB=0, desc=out}
  n44 = ConstantBiasScale(None=@n43) {bias=-0.75, scale=4}
  n45 = Multiply(A=NamedRerouteUsage(), B=Multiply(A=NamedRerouteUsage(), B=Saturate(None=@n37))) {desc=out}
  n46 = OneMinus(None=@n30)
  n47 = OneMinus(None=@n34)
  n48 = Max(A=Saturate(None=ConstantBiasScale(None=@n46) {bias=-0.975, scale=40}), B=Saturate(None=ConstantBiasScale(None=@n47) {bias=-0.975, scale=40})) {desc=Out}
  n49 = LinearInterpolate(A=@n48, Alpha=@n42) {constB=0, desc=TopLinePatch}
  n50 = Saturate(None=Max(A=Add(A=@n44, B=@n45), B=@n49))
  n51 = LinearInterpolate(A=@n2, Alpha=@n12) {constB=0, desc=Pink Lines}
  n52 = ConstantBiasScale(None=@n43) {bias=-0.92, scale=100}
  n53 = Saturate(None=Max(A=@n49, B=Saturate(None=@n52)))
  n54 = LinearInterpolate(
   A=LinearInterpolate(A=LinearInterpolate(A=@n1.RGB, B=@n2, Alpha=@n12) {constB=0, desc=Base Color}, B=@n40, Alpha=@n50) {constB=0, desc=Base Color}
   B=@n51
   Alpha=@n53
  ) {desc=Base Color}
  n55 = Saturate(None=ConstantBiasScale(None=MaterialFunctionCall!(Texture Coordinates, Tiling X, Tiling Y, PanSpeedX, PanSpeedY, Angle, RotateSpeed, Balance, Chaos) {materialFunction=/VehicleCosmetics/Materials/MaterialFunctions/MF_Noisemaker_Perlin}.Perlin2) {bias=-0.75, scale=50})
  n56 = ConstantBiasScale(None=@n43) {bias=-0.95, scale=50}
  n57 = Saturate(None=Max(A=Saturate(None=@n56), B=@n49))
  n58 = LinearInterpolate(A=@n13, B=Constant3Vector() {constant=<rgba 0.8308 0.807 0 1>}, Alpha=Saturate(None=Subtract(A=@n55, B=@n57)))
  n59 = NamedRerouteUsage()
  n60 = Multiply(A=Abs(None=@n59)) {constB=500}
  n61 = Clamp(None=Multiply(A=@n59) {constB=0.125}) {minDefault=0.02}
  n62 = Saturate(None=ConstantBiasScale(None=Add(A=@n61, B=@n43)) {bias=-0.975, scale=25})
  n63 = Clamp(None=Multiply(A=@n59) {constB=0.5}) {minDefault=0.02}
  n64 = Saturate(None=ConstantBiasScale(None=Add(A=@n63, B=@n49)) {bias=-0.75, scale=2})
  n65 = Saturate(None=Multiply(A=Clamp(None=@n60) {minDefault=0.25}, B=Max(A=@n62, B=@n64)))
  n66 = Constant3Vector() {constant=<rgba 0.8632 0.8632 0.8632 1>}
  n67 = ConstantBiasScale(None=@n41) {bias=-0.8, scale=25}
  n68 = ConstantBiasScale(None=NamedRerouteUsage()) {bias=-0.46, scale=200}
  n69 = Multiply(A=Max(A=Saturate(None=@n67), B=@n48), B=Multiply(A=Saturate(None=@n68), B=NamedRerouteUsage()))
  n70 = LinearInterpolate(A=Multiply(A=@n2, B=@n12) {desc=Emissive}, Alpha=@n43) {constB=0, desc=Emissive}
  n71 = LinearInterpolate(
   A=LinearInterpolate(A=LinearInterpolate(A=@n70, B=@n40, Alpha=@n50) {constB=0, desc=Emissive}, B=@n51, Alpha=@n53) {desc=Emissive}
   B=@n58
   Alpha=@n65
  ) {desc=Emissive}
  n72 = ScalarParameter() {defaultValue=1, parameterName=EmissiveIntensity}
  n73 = SetMaterialAttributes!(
   MaterialAttributes
   BaseColor=LinearInterpolate(A=LinearInterpolate(A=@n54, B=@n58, Alpha=@n65) {desc=Base Color}, B=@n66, Alpha=@n69) {desc=Base Color}
   Metallic=Constant() {r=0.2}
   Roughness=Constant() {r=0.2}
   EmissiveColor=LinearInterpolate(A=Multiply(A=@n71, B=@n72) {desc=Emissive}, B=@n66, Alpha=@n69) {desc=Emissive}
   Normal
   AmbientOcclusion
  ) {attributeSetTypes [
    #1
    #2
    #3
    #4
    #5
    #6
   ]}
  FunctionOutput(None=@n73)
  n74 = MaterialFunctionCall!(FakeBeatIntensity, BeatSyncIntensity=ScalarParameter() {defaultValue=1, parameterName=BeatSyncIntensity}) {materialFunction=/VehicleCosmetics/SharedMaterials/MaterialFunctions/MF_AudioBeatSync}
  n75 = Max(A=ComponentMask(None=@n74.Spectrum) {b=true}, B=ComponentMask(None=@n74.Spectrum) {a=true})
  n76 = Add(A=ComponentMask(None=@n74.Spectrum) {r=true}, B=ComponentMask(None=@n74.Spectrum) {g=true})
  NamedRerouteDeclaration(None=Add(A=@n75, B=@n76)) {name=BeatSync, nodeColor=<rgba 0.3594 0 1 1>}
  n77 = Add(A=ComponentMask(None=TextureCoordinate() {uTiling=2, vTiling=4}) {r=true}) {constB=0.1}
  n78 = Subtract(A=@n77, B=Floor(None=@n77))
  n79 = Power(Base=Abs(None=ConstantBiasScale(None=@n78) {bias=-0.5, scale=2})) {constExponent=0.3}
  NamedRerouteDeclaration(None=Multiply(A=Saturate(None=@n79)) {constB=0.5}) {name=WarpEdges, nodeColor=<rgba 0 0.6016 1 1>}
  NamedRerouteDeclaration(None=@n1.A) {name=ColorMapAlpha, nodeColor=<rgba 0.9219 1 0 1>}
  NamedRerouteDeclaration(None=@n76) {name=BeatSync Low, nodeColor=<rgba 0 0.2957 0.0475 1>}
  NamedRerouteDeclaration(None=@n75) {name=BeatSync Hi, nodeColor=<rgba 0.724 0.0103 0 1>}
 ]}
